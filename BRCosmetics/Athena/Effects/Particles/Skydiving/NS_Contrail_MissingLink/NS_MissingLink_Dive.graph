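# Niagara skydiving dive contrail for the MissingLink cosmetic, showing hotdog link meshes attached to the player with loose hotdogs shedding off.
# HotdogLinks_Mesh bursts four long-lived meshes, pins each to a skeletal bone every update, orients them to the owner, flips alternates and drives noise material parameters.
# Hotdog_Particles spawns short-lived tumbling meshes from bones with launch velocity, slight reversed inherited velocity, drag and a pop-in, pop-out scale curve.
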NiagaraSystem /BRCosmetics/Athena/Effects/Particles/Skydiving/NS_Contrail_MissingLink/NS_MissingLink_Dive
dependencies {
 usedDataInterfaces [&NiagaraDataInterfaceVectorCurve]
 usedDynamicInputs []
 usedModules [/Game/Effects/Niagara/Modules/SkeletalMesh/FN_SkeletalBoneData, &SystemState]
}
emitters {
 HotdogLinks_Mesh {
  emitterSpawnScript [&FN_PreserveSkydiveUserVariables() {module=PreserveUserVariables}]
  emitterUpdateScript [
   &EmitterState()
   &SpawnBurst_Instantaneous(Spawn Count=4, Use Spawn Probability=false)
  ]
  particleSpawnScript [
   /Game/Effects/Niagara/Modules/Groups/FN_SetSpawnGroup_v1/FN_SetSpawnGroup(Emitter Bone Count=4)
   /Niagara/Modules/Spawn/Initialization/InitializeParticle(
    Lifetime=100
    Mesh Scale=<0.5 0.5 0.5>
    Module.Write Color=true
    Module.Write Lifetime=true
    Module.Write Mass=true
    Module.Write Material Random=true
    Module.Write Position=false
    Module.Write Scale=true
    Module.Write SpriteRotation=false
    Module.Write SpriteSize=false
    Position=&SimulationPosition()
   )
  ]
  particleUpdateScript [
   &ParticleState()
   &SampleSkeletalMeshSkeleton(Mesh:NiagaraDataInterfaceSkeletalMesh=$User.SkeletalMesh, Specific Bone Index:NiagaraInt32=$Particles.SpawnGroup)
   /Niagara/Modules/Spawn/Location/SkeletalMeshSkeletonLocation(Module.Inherit Velocity=false, Module.Use Sampled Bone as Ribbon ID=true)
   /BRCosmetics/Athena/Effects/Particles/Skydiving/NS_Contrail_AmmoBelt/Module/FN_OrientMeshToEngineOwner(
    Direction:Vector3f=&FN_FrontEndVector3Modifier(Gameplay=<0 0 -1>, Front End=<0 0 -1>)
    Mesh Forward Vector:Vector3f=<0 1 0>
    Vector:Vector3f=&TransformVector(Vector=$Engine.Owner.SystemYAxis, Source Space=::World, Destination Space=::Local)
   )
   /BRCosmetics/Athena/Effects/Particles/Skydiving/NS_Contrail_AmmoBelt/Module/FN_QuaternionSpin(
    Angle=HlslExpression() {hlslExpression="fmod(Particles.SpawnGroup, 2.0)*3.1415"}
    Axis:Vector3f=<0 0 1>
    Quaternion:Quat4f=$Particles.FN_OrientMeshToEngineOwner.Quaternion
    Set Mesh Orientation=true
   )
   /Game/Effects/Niagara/Modules/Position/FN_OffsetPositionByOrientationAxis(Offset Axis:Vector3f=<0 1 0>, Y Symmetry=true) {module=OffsetPositionByMeshOrientation}
   /Game/Effects/Niagara/Modules/Materials/FN_MeshMaterial(Amplitude=60, Falloff=60, Frequency=300, Scale Noise X Axis=2, Scale Noise Y Axis=0.3, Scale Noise Z Axis=0, Speed=0.8)
   &SolveForcesAndVelocity(Clamp Velocity=false, Limit Acceleration=false) {enabled=false}
  ]
  rendererClasses [&NiagaraMeshRendererProperties]
  renderers [{rendererClass=&NiagaraMeshRendererProperties, rendererIndex=0}]
 }
 Hotdog_Particles {
  emitterSpawnScript [&FN_PreserveSkydiveUserVariables() {module=PreserveUserVariables}]
  emitterUpdateScript [
   &EmitterState()
   &SpawnRate(SpawnRate=&UniformRangedFloat(Minimum=0.6, Maximum=1), Use Spawn Probability=false)
  ]
  particleSpawnScript [
   &FN_SetSpawnGroup(Emitter Bone Count=2)
   &InitializeParticle(
    Lifetime=&UniformRangedFloat(Minimum=0.9, Maximum=1.3)
    Mesh Scale=<0.3 0.3 0.3>
    Module.Write Color=true
    Module.Write Lifetime=true
    Module.Write Mass=true
    Module.Write Material Random=true
    Module.Write Position=false
    Module.Write Scale=true
    Module.Write SpriteRotation=false
    Module.Write SpriteSize=false
    Position=&SimulationPosition()
   )
   &SampleSkeletalMeshSkeleton(Mesh:NiagaraDataInterfaceSkeletalMesh=$User.SkeletalMesh, Specific Bone Index:NiagaraInt32=$Particles.SpawnGroup)
   &SkeletalMeshSkeletonLocation(Module.Inherit Velocity=false, Module.Use Sampled Bone as Ribbon ID=true)
   /Niagara/Modules/Spawn/Orientation/V2/InitialMeshOrientation(
    Rotation=&UniformRangedVector(Minimum=<0 0 0>, Maximum=<1 1 1>)
    Use Orientation Vector=true
    Use Rotation Vector=true
   )
   &AddVelocity(Velocity=<0 0 350>)
   &InheritVelocity(Module.Limit Speed=true, Velocity Scale=<-0.1 -0.1 -0.1>)
  ]
  particleUpdateScript [
   &ParticleState()
   &ScaleMeshSize(Scale Factor=&ScaleFloatByCurve(
     InitialValue:Vector3f=<1 1 1>
     ScaleCurve:NiagaraDataInterfaceVectorCurve=DataInterface() {propertyValues {
       XCurve {keys [
         {interpMode=RCIM_Cubic, tangentMode=RCTM_User, time=0, value=0}
         {arriveTangent=0.11, interpMode=RCIM_Cubic, leaveTangent=0.11, tangentMode=RCTM_User, time=0.1017, value=2.3268}
         {interpMode=RCIM_Cubic, tangentMode=RCTM_User, time=0.1791, value=1.2434}
         {arriveTangent=3.0207, interpMode=RCIM_Cubic, leaveTangent=3.0207, tangentMode=RCTM_User, time=0.7564, value=1.5342}
         {interpMode=RCIM_Cubic, tangentMode=RCTM_User, time=0.7963, value=2.2724}
         {interpMode=RCIM_Cubic, tangentMode=RCTM_User, time=1, value=0}
        ]}
       YCurve {keys [<at 0.0026 0.0926>,<at 0.1002 2.3488>,<at 0.1906 1.2434>,<at 0.7635 1.4798>,<at 0.804 2.3299>,<at 1 0>]}
       ZCurve {keys [<at -0.0012 0.0926>,<at 0.1002 2.3268>,<at 0.1982 1.3009>,<at 0.7597 1.4223>,<at 0.8193 2.2724>,<at 1.0038 0.1501>]}
      }}
     CurveIndex:NiagaraFloat=$Particles.NormalizedAge
    ))
   &Drag(Use Linear Drag=true, Use Rotational Drag=false)
   /Niagara/Modules/Update/Orientation/V2/UpdateMeshOrientation(Rotation Rate=0.6, Rotation Vector:Vector3f=<1 1 0>)
   &SolveForcesAndVelocity(Clamp Velocity=false, Limit Acceleration=false)
  ]
  rendererClasses [&NiagaraMeshRendererProperties]
  renderers [{rendererClass=&NiagaraMeshRendererProperties, rendererIndex=0}]
 }
}
summary {
 systemName=NS_MissingLink_Dive
 emitters [
  {emitterName=HotdogLinks_Mesh, rendererClasses [&NiagaraMeshRendererProperties]}
  {emitterName=Hotdog_Particles, rendererClasses [&NiagaraMeshRendererProperties]}
 ]
 userVariables [
  User.bIsFrontEnd {type=NiagaraBool}
  User.bIsFrontEndPreview {type=NiagaraBool}
  User.PlayerSpeed {type=NiagaraFloat, defaultValue=0, description=""}
  User.SkeletalMesh {
   type=NiagaraDataInterfaceSkeletalMesh
   defaultValue {
    struct=NiagaraExt_VariableValue_DataInterface
    value {dataInterface=&self:NiagaraDataInterfaceSkeletalMesh_0}
   }
  }
 ]
}
